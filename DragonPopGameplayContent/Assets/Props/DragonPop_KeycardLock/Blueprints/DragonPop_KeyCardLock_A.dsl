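; Keycard lock Blueprint that extends a parent interactable lock and shows its state on the lock's screen material.
; BeginPlay and SetLock refresh the screen, using scalar parameters to show granted or rejected access, the light colour and a game-time stamp.
; On unlock, it applies gameplay effects to the last interacting pawn and broadcasts the faction door-open custom verb.
; DEBUG events duplicate the unlock and reject screen logic.
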(event Custom|DEBUG_ScreenReject
  (bind _lock (Variables|Default|GetLock))
  ; DEBUG & Lock Visuals
  (Rendering|Material|SetScalarParameterValueonMaterials _lock "BP_GrantAccessIndex" 0.0)
  (Rendering|Material|SetScalarParameterValueonMaterials _lock "BP_TurnLightGreen" 0.0)
  (Rendering|Material|SetScalarParameterValueonMaterials _lock "BP_GetTime" (Utilities|Time|GetGameTimeinSeconds)))

(event Custom|CustomEvent_RejectAcessScreen
  (bind _lock (Variables|Default|GetLock))
  ; DEBUG & Lock Visuals
  (Rendering|Material|SetScalarParameterValueonMaterials _lock "BP_GrantAccessIndex" 0.0)
  (Rendering|Material|SetScalarParameterValueonMaterials _lock "BP_TurnLightGreen" 0.0)
  (Rendering|Material|SetScalarParameterValueonMaterials _lock "BP_GetTime" (Utilities|Time|GetGameTimeinSeconds)))

(event Custom|DEBUG_ScreenUnlock
  (bind _lock (Variables|Default|GetLock))
  ; DEBUG & Lock Visuals
  (Rendering|Material|SetScalarParameterValueonMaterials _lock "BP_GrantAccessIndex" 1.0)
  (Rendering|Material|SetScalarParameterValueonMaterials _lock "BP_TurnLightGreen" 1.0)
  (Rendering|Material|SetScalarParameterValueonMaterials _lock "BP_GetTime" (Utilities|Time|GetGameTimeinSeconds)))

(event Custom|CustomEvent_UnlockScreen
  (bind _lock (Variables|Default|GetLock))
  ; DEBUG & Lock Visuals
  (Rendering|Material|SetScalarParameterValueonMaterials _lock "BP_GrantAccessIndex" 1.0)
  (Rendering|Material|SetScalarParameterValueonMaterials _lock "BP_TurnLightGreen" 1.0)
  (Rendering|Material|SetScalarParameterValueonMaterials _lock "BP_GetTime" (Utilities|Time|GetGameTimeinSeconds)))

(event EventOnUnlocked
  (|CallParentFunction)
  (Fortnite|ApplyGameplayEffectsToActor (Variables|LockParentInteract|GetLastInteractingPawn) (Variables|Default|GetEffects%20to%20Apply) (Transformation|GetActorLocation (Variables|LockParentInteract|GetLastInteractingPawn)))
  ;[1 1 1] Send Custom Verb
  (CustomVerb|BroadcastCustomVerbMessage (CustomVerb|MakeCustomVerbMessage (Fortnite|GetFortPlayerControllerfromActor 0) "(TagName=\"CustomVerb.Faction.Door.Open\")")))

(event EventSetLock
  (params (LockState :type Enum_Athena_Lock%20Enum :default "NewEnumerator0"))
  (|CallParentFunction LockState)
  (CallFunction|UpdateLockScreenMaterialLook))

(event Custom|UpdateLockScreenMaterialLook
  (if (== (Variables|LockParent|GetLockState) 2)
    (CallFunction|CustomEventUnlockScreen)
    (else
      (CallFunction|CustomEventRejectAcessScreen))))

(event EventBeginPlay
  (|CallParentFunction)
  (CallFunction|UpdateLockScreenMaterialLook))

(event Collision|EventActorBeginOverlap
  (params (OtherActor :type Actor%20Object%20Reference)))

(event EventTick
  (params (DeltaSeconds :type Float%20%28single-precision%29)))

(fn ConstructionScript
  (|CallParentFunction))
