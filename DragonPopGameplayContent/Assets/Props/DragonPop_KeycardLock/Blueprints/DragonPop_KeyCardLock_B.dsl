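; Keycard lock prop for DragonPop content, built on a parent lock Blueprint that handles interaction and unlocking.
; On unlock it runs the parent logic, then applies its configured gameplay effects to the last interacting pawn at that pawn's location.
; It then broadcasts the faction crate-open custom verb, though the player controller lookup for that message has no actor wired in.
; BeginPlay, overlap, Tick and the construction script add no logic of their own.
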(event EventOnUnlocked
  (|CallParentFunction)
  (Fortnite|ApplyGameplayEffectsToActor (Variables|LockParentInteract|GetLastInteractingPawn) (Variables|Default|GetEffects%20to%20Apply) (Transformation|GetActorLocation (Variables|LockParentInteract|GetLastInteractingPawn)))
  ;[1 1 1] Send Custom Verb
  (CustomVerb|BroadcastCustomVerbMessage (CustomVerb|MakeCustomVerbMessage (Fortnite|GetFortPlayerControllerfromActor 0) "(TagName=\"CustomVerb.Faction.Crate.Open\")")))

(event EventBeginPlay
  (|CallParentFunction) ; This node is disabled and will not be called.\nDrag off pins to build functionality.
)

(event Collision|EventActorBeginOverlap
  (params (OtherActor :type Actor%20Object%20Reference)))

(event EventTick
  (params (DeltaSeconds :type Float%20%28single-precision%29)))

(fn ConstructionScript
  (|CallParentFunction))
