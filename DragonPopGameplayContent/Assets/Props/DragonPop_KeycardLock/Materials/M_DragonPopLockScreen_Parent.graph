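# Parent material for the DragonPop keycard lock screen, drawing an animated CRT-style lock icon for access granted or denied.
# BP_GetTime sets the animation start time and BP_GrantAccessIndex blends between denied and unlock-then-checkmark timelines.
# These drive a channel-packed 4x4 flipbook with color changes, a fading denied wobble and timed glitch UVs.
# Scanlines, CRT texture and a circular mask shape the emissive, which divides by eye adaptation.
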Material /DragonPopGameplayContent/Assets/Props/DragonPop_KeycardLock/Materials/M_DragonPopLockScreen_Parent
graph {
 properties {
  bAutomaticallySetUsageInEditor=false
  bUseMaterialAttributes=true
  bUsedWithClothing=true
  bUsedWithNanite=true
  usedShadingModels=Default Lit
 }
 nodes [
  n1 = MaterialFunctionCall!(ScanLine_Tiling_1, Scanline_PannerSpeed_1, ScanLine_Tiling, Scanline_PannerSpeed, UVs) {materialFunction=/Game/Weapons/WeaponSkins/MaterialFunction/ScanLines}
  n2 = TextureSample(UVs=Multiply(A=TextureCoordinate(), B=ScalarParameter() {defaultValue=80, group=AnimatedScreen, parameterName=CRT_Scale}) {constB=80}) {
   samplerSource=SSM_Wrap_WorldGroupSettings
   samplerType=SAMPLERTYPE_LinearColor
   texture=/Game/Accessories/FORT_Backpacks/Backpack_M_MED_Jumbotron_S20/Textures/T_Jumbotron_CRT
  }
  n3 = ScalarParameter() {group=AnimatedScreen, parameterName=CRT_Desaturation}
  n4 = LinearInterpolate(
   A=ScalarParameter() {defaultValue=0.1, group=AnimatedScreen, parameterName=Screen_MinEmissive}
   B=NamedRerouteUsage()
   Alpha=NamedRerouteUsage()
  ) {constA=0.1}
  n5 = LinearInterpolate(
   A=NamedRerouteUsage()
   B=LinearInterpolate(
    A=ScalarParameter() {defaultValue=0.1, group=AnimatedScreen, parameterName=Screen_MinEmissive}
    B=NamedRerouteUsage()
    Alpha=NamedRerouteUsage()
   ) {constA=0.1}
   Alpha=NamedRerouteUsage()
  )
  n6 = ScalarParameter() {group=AnimatedScreen, parameterName=BP_GrantAccessIndex, sortPriority=10000}
  n7 = Multiply(
   A=Saturate(None=Add(A=@n1.Result) {constB=0.45})
   B=Multiply(A=Desaturation(None=@n2.RGB, Fraction=@n3), B=LinearInterpolate(A=@n4, B=@n5, Alpha=@n6))
  )
  n8 = Power(Base=@n7) {constExponent=4}
  n9 = ScalarParameter() {defaultValue=100, group=AnimatedScreen, parameterName=EmissiveBrightness}
  n10 = Desaturation(
   None=Add(A=@n7, B=Multiply(A=@n8, B=@n9))
   Fraction=ScalarParameter() {defaultValue=-2, group=AnimatedScreen, parameterName=EmissiveDesaturation}
  )
  n11 = EyeAdaptation()
  n12 = TextureCoordinate()
  n13 = Constant2Vector() {g=0.5, r=0.5}
  MakeMaterialAttributes!(
   BaseColor=ScalarParameter() {defaultValue=0.025, group=BasicPBR, parameterName=BaseColor_Intensity}
   Metallic
   Specular=ScalarParameter() {defaultValue=0.5, group=BasicPBR, parameterName=Specular_Amount}
   Roughness=ScalarParameter() {defaultValue=0.5, group=BasicPBR, parameterName=Roughness_Amount}
   Anisotropy
   EmissiveColor=Multiply(A=Divide(A=@n10, B=@n11.EyeAdaptation), B=SphereMask(A=@n12, B=@n13) {attenuationRadius=0.7, hardnessPercent=0.1})
   Opacity
   OpacityMask
   Normal
   Tangent
   WorldPositionOffset
   SubsurfaceColor
   ClearCoat
   ClearCoatRoughness
   AmbientOcclusion
   Refraction
   CustomizedUVs_0
   CustomizedUVs_1
   CustomizedUVs_2
   CustomizedUVs_3
   CustomizedUVs_4
   CustomizedUVs_5
   CustomizedUVs_6
   CustomizedUVs_7
   PixelDepthOffset
   ShadingModel
   Displacement
  )
  n14 = Constant() {r=4}
  n15 = NamedRerouteUsage()
  n16 = ScalarParameter() {group=AnimatedScreen, parameterName=BP_GrantAccessIndex, sortPriority=10000}
  n17 = Sine(None=Multiply(A=Time()) {constB=6})
  n18 = Multiply(A=OneMinus(None=@n16), B=Multiply(A=Multiply(A=@n17) {constB=0.2}, B=NamedRerouteUsage()))
  n19 = LinearInterpolate(A=NamedRerouteUsage(), B=NamedRerouteUsage(), Alpha=@n16)
  n20 = TextureSample(UVs=MaterialFunctionCall("Animation  Phase (0-1)"=@n19, Number of Rows=@n14, Number of Columns=@n14, UVs=Rotator(Coordinate=@n15, Time=@n18)) {materialFunction=&FlipBook}.UVs) {
   samplerSource=SSM_Wrap_WorldGroupSettings
   samplerType=SAMPLERTYPE_LinearColor
   texture=/DragonPopGameplayContent/Gameplay/Textures/T_LockIconsAnimation_Outlaw
  }
  NamedRerouteDeclaration(None=@n20.R) {name=Flipbook_LockOpening, nodeColor=<rgba 0 0.0729 1 1>}
  NamedRerouteDeclaration(None=@n20.G) {name=Flipbook_LockDenied, nodeColor=<rgba 1 0.0061 0.0002 1>}
  n21 = VectorParameter() {group=AnimatedScreen, parameterName=LockColor_Closed, sortPriority=10, defaultValue=<rgba 1 0.0222 0 1>}
  n22 = VectorParameter() {group=AnimatedScreen, parameterName=LockColor_Opened, sortPriority=10, defaultValue=<rgba 0.1426 0.7448 0.0295 1>}
  n23 = Constant() {r=-2}
  n24 = Multiply(A=Frac(None=@n19)) {constB=2}
  n25 = LinearInterpolate(
   A=NamedRerouteDeclaration(None=@n21.RGB) {name=ScreenColor_Red, nodeColor=<rgba 1 0.4101 0.0376 1>}
   B=NamedRerouteDeclaration(None=Desaturation(None=@n22.RGB, Fraction=@n23)) {name=ScreenColor_Green, nodeColor=<rgba 0.182 1 0.0917 1>}
   Alpha=Saturate(None=Floor(None=@n24))
  )
  n26 = NamedRerouteUsage()
  NamedRerouteDeclaration(None=LinearInterpolate(A=ScalarParameter() {defaultValue=0.1, group=AnimatedScreen, parameterName=Screen_MinEmissive}, B=Multiply(A=@n25, B=@n26), Alpha=@n26) {constA=0.1}) {name=LockColorsWhileOpening, nodeColor=<rgba 1 0 0.2344 1>}
  n27 = NamedRerouteUsage()
  n28 = Divide(A=@n27) {constB=16}
  n29 = Subtract(A=Saturate(None=Multiply(A=@n28) {constB=20})) {constB=0.5}
  n30 = SmoothStep(Value=Abs(None=Multiply(A=@n29) {constB=2})) {constMin=0.9}
  n31 = ScalarParameter() {group=AnimatedScreen, parameterName=BP_GrantAccessIndex, sortPriority=10000}
  n32 = Subtract(
   A=@n27
   B=LinearInterpolate(Alpha=ScalarParameter() {group=AnimatedScreen, parameterName=BP_GrantAccessIndex, sortPriority=10000}) {constA=1.4, constB=1.8}
  )
  n33 = Saturate(None=Multiply(A=Divide(A=@n32) {constB=16}) {constB=15})
  n34 = Abs(None=Multiply(A=Subtract(A=@n33) {constB=0.5}) {constB=2})
  n35 = LinearInterpolate(
   A=MaterialFunctionCall!(UV) {materialFunction=/Game/Global/Materials/MaterialGym/Materials/HolographicTexture/MF_HolographicTextureGlitch_Grid}.Glitched UV
   B=TextureCoordinate()
   Alpha=Min(A=LinearInterpolate(A=@n30, Alpha=@n31), B=SmoothStep(Value=@n34) {constMin=0.9})
  )
  NamedRerouteDeclaration(None=@n35) {name=GlitchUVs, nodeColor=<rgba 1 0.3984 0 1>}
  NamedRerouteDeclaration(None=@n20.B) {name=Flipbook_LockOpened, nodeColor=<rgba 0.5032 1 0 1>}
  n36 = Time()
  n37 = ScalarParameter() {defaultValue=999999995904, group=AnimatedScreen, parameterName=BP_GetTime, sortPriority=10000}
  n38 = Divide(A=NamedRerouteDeclaration(None=Subtract(A=@n36, B=@n37)) {name=Blueprint_Timing, nodeColor=<rgba 0 1 0.9766 1>}) {constB=16}
  n39 = Multiply(A=@n38) {constB=24, desc=Flipbook Frame Rate}
  n40 = Subtract(A=NamedRerouteUsage(), B=ScalarParameter() {defaultValue=2, group=AnimatedScreen, parameterName=CheckmarkAnimationDelay, sortPriority=20}) {constB=2}
  n41 = Multiply(A=Divide(A=@n40) {constB=16}) {constB=24, desc=Flipbook Frame Rate}
  n42 = Constant() {r=1}
  n43 = If(A=@n40, "A > B"=@n42, "A == B"=@n42, "A < B"=Constant())
  NamedRerouteDeclaration(None=LinearInterpolate(
    A=Clamp(None=@n39) {maxDefault=0.95}
    B=Clamp(None=@n41) {maxDefault=0.95}
    Alpha=NamedRerouteDeclaration(None=@n43) {name=ActivateCheckmarkAnimation, nodeColor=<rgba 1 0.0661 0.4918 1>}
   )) {name=Timing_Unlock, nodeColor=<rgba 0.3359 0 1 1>}
  n44 = Multiply(A=@n38) {constB=16, desc=Flipbook Frame Rate}
  n45 = NamedRerouteUsage()
  n46 = Subtract(A=@n45) {constB=2}
  n47 = Divide(A=@n46) {constB=16}
  n48 = Clamp(None=OneMinus(None=Multiply(A=@n47) {constB=60})) {maxDefault=0.95}
  n49 = Multiply(A=@n48, B=@n48)
  n50 = Constant() {r=1}
  n51 = Constant()
  NamedRerouteDeclaration(None=LinearInterpolate(A=Clamp(None=@n44) {maxDefault=0.95}, B=Multiply(A=@n49, B=@n49), Alpha=If(A=@n46, "A > B"=@n50, "A == B"=@n50, "A < B"=@n51))) {name=Timing_AccessDenied, nodeColor=<rgba 0.5625 0.3244 1 1>}
  n52 = OneMinus(None=Multiply(A=@n47) {constB=20})
  n53 = Subtract(A=@n45) {constB=0.5}
  NamedRerouteDeclaration(None=Multiply(A=Clamp(None=@n52) {maxDefault=0.95}, B=If(A=@n53, "A > B"=@n50, "A == B"=@n50, "A < B"=@n51))) {name=FadeOutRotationAnimation, nodeColor=<rgba 1 0 0.3047 1>}
  Reroute(None=@n35)
  Reroute(None=@n43)
 ]
}
